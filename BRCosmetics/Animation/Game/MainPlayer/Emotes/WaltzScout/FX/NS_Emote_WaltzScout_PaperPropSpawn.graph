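# Niagara system for the WaltzScout player emote, named as the effect for its paper prop spawn.
# It holds three emitters, SparklesLWE, AirglowLWE and SmokeLWE, whose suffix suggests lightweight emitters; no renderers are listed for them.
# The system script uses SystemState and a variable assignment, with float-to-int and integer-multiply dynamic inputs.
# It exposes float user parameters BurstIndex and NormalizedNotifyProgress, the latter apparently meant to be fed by an animation notify.
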NiagaraSystem /BRCosmetics/Animation/Game/MainPlayer/Emotes/WaltzScout/FX/NS_Emote_WaltzScout_PaperPropSpawn
dependencies {
 usedDataInterfaces []
 usedDynamicInputs [&Multiply_Int, &MakeIntFromFloat]
 usedModules [
  &self:SystemSpawnScript.SystemScriptSource.SystemScriptGraph.NiagaraNodeAssignment_0.SetVariables_#1
  &SystemState
 ]
}
emitters {AirglowLWE {rendererClasses [], renderers []}, SmokeLWE {rendererClasses [], renderers []}, SparklesLWE {rendererClasses [], renderers []}}
summary {
 systemName=NS_Emote_WaltzScout_PaperPropSpawn
 emitters [{emitterName=SparklesLWE, rendererClasses []}, {emitterName=AirglowLWE, rendererClasses []}, {emitterName=SmokeLWE, rendererClasses []}]
 userVariables [
  User.BurstIndex {type=NiagaraFloat, defaultValue=1, description=""}
  User.NormalizedNotifyProgress {type=NiagaraFloat, defaultValue=0, description=""}
 ]
}
